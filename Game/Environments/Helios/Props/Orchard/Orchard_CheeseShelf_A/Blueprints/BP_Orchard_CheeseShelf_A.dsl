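; BP_Orchard_CheeseShelf_A is an Orchard environment prop, a cheese shelf with three stack meshes (A, B and C) that get randomized tint colors.
; The construction script calls its parent, then loops over the three stack components and passes each a random entry from the TintColor_Cheese array via SetColoronSecondaryMesh.
; BeginPlay, overlap and Tick events are empty, so the graphs shown add no runtime behavior.
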(event EventBeginPlay)

(event Collision|EventActorBeginOverlap
  (params (OtherActor :type Actor%20Object%20Reference)))

(event EventTick
  (params (DeltaSeconds :type Float%20%28single-precision%29)))

(fn ConstructionScript
  (|CallParentFunction)
  (for _array_element (Utilities|Array|MakeArray (Variables|Default|GetSM_Orchard_CheeseShelfStack_A) (Variables|Default|GetSM_Orchard_CheeseShelfStack_B) (Variables|Default|GetSM_Orchard_CheeseShelfStack_C))
    (Default|SetColoronSecondaryMesh (Utilities|Array|RandomArrayItem (Variables|Default|GetTintColor_Cheese)) _array_element)))
